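# Physics asset for the Retro Wheels pickaxe mesh, covering a small dangling charm on the weapon.
# Three small capsule bodies sit on the base charm bone and two dynamic charm bones.
# Two constraints link them in a chain from the base outward, each with equal swing limits on both axes and a zero twist limit.
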PhysicsAsset /BRCosmetics/Weapons/FORT_Melee/Pickaxe_RetroWheels/Meshes/RetroWheels_Axe_Physics
bodies {
 base_charm [{shapeType=Capsule, length=0.505, radius=0.505, shapeName=base_charm_capsule}]
 dyn_charm_01_m [{shapeType=Capsule, length=0.5, radius=0.5, shapeName=Nonedyn_charm_01_m_capsule, center=<0.8124 0 0>, rotation=<rot -90 0 0>}]
 dyn_charm_02_m [{shapeType=Capsule, length=0.5, radius=0.5, shapeName=dyn_charm_02_m_capsule, center=<0.7248 0 0>, rotation=<rot -90 180 180>}]
}
constraints [
 {bone1Name=dyn_charm_02_m, bone2Name=dyn_charm_01_m, swing1LimitDegrees=65, swing2LimitDegrees=65, twistLimitDegrees=0}
 {bone1Name=dyn_charm_01_m, bone2Name=base_charm, swing1LimitDegrees=65, swing2LimitDegrees=65, twistLimitDegrees=0}
]
